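# Niagara effect for the Dire vehicle wheel cosmetic, built from three GPU sprite emitters offset along local Y.
# Lightnings spawn continuously across a flat disc, facing local Y, with randomized material parameters and growing size.
# Rays bursts a blue, wheel-facing sprite that persists because its particle state update is disabled.
# Smoke emits drifting, expanding flipbook puffs; skeletal mesh placement is disabled throughout, as are the Lightnings and Smoke color updates.
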NiagaraSystem /VehicleCosmetics/Wheels/Dire/PS/PS_Wheel_Dire
dependencies {
 usedDataInterfaces [&NiagaraDataInterfaceCurve, &NiagaraDataInterfaceVector2DCurve]
 usedDynamicInputs []
 usedModules [&SystemState]
}
emitters {
 Lightnings {
  simTarget=GPUComputeSim
  emitterUpdateScript [
   &EmitterState(Life Cycle Mode=::NewEnumerator1, Loop Duration=1, UseLoopDelay=false)
   &SpawnRate(SpawnRate=4, Use Spawn Probability=false)
   &SpawnBurst_Instantaneous(
    Spawn Count=&RandomRangeInt(Minimum=-1, Maximum=1)
    Use Spawn Probability=false
    UseLoopCountLimit=false
   )
  ]
  particleSpawnScript [
   &SkeletalMeshLocation(
    Bone Sampling Mode=::NewEnumerator4
    MaskSpawnGroup=false
    Position Sampling=::NewEnumerator2
    Sampled Position Offset=<0 32 0>
    Skeletal Mesh=$System.Skeletal Mesh
   ) {enabled=false}
   &InitializeParticle(
    Color Mode=::NewEnumerator0
    Lifetime=&RandomRangeFloat(Minimum=0.25, Maximum=0.5)
    Position=<0 32 0>
    Position Mode=::NewEnumerator1
    Position Offset Coordinate Space=::Local
    Sprite Rotation Angle=&RandomRangeFloat(Minimum=0, Maximum=1)
    Sprite Rotation Mode=::NewEnumerator1
    Sprite Size Mode=::NewEnumerator3
    Uniform Sprite Size=&RandomRangeFloat(Minimum=3.2, Maximum=16)
    UsePositionOffset=true
    Write Material Random=true
   ) {module=InitializeParticle001}
   /SetVariables(Particles.SubImageIndex=&FloatFromCurve(FloatCurve=DataInterface() {propertyValues {Curve {keys [<at 0 0 RCIM_Cubic>,<at 1 8 RCIM_Cubic>]}}})) {isSetParameters=true, module=SetVariables_#1}
   &DynamicMaterialParameters(
    Index 0 Param 1=&RandomRangeFloat(Minimum=0, Maximum=1)
    Index 0 Param 2=&RandomRangeFloat(Minimum=10, Maximum=30)
    Index 0 Param 4=&RandomRangeFloat(Minimum=0, Maximum=0.25)
    Module.Index0Param0Write=true
    Module.Index0Param1Write=true
    Module.Index0Param2Write=true
    Module.Index0Param3Write=true
   )
   &ShapeLocation(
    Cylinder Height=0
    Cylinder Height Midpoint=0
    Cylinder Radius=24
    MaskSpawnGroup=false
    Offset Coordinate Space=::Local
    Offset Mode=::NewEnumerator0
    Override Seed=false
    Rotation Axis=<0 1 0>
    Rotation Mode=::NewEnumerator1
    Shape Primitive=::NewEnumerator1
    Surface Only=false
    Surface Only Band Thickness=1
    UseOwnerScale=true
   )
   &SpriteFacingAndAlignment(Facing Coordinate Space=::Local, Sprite Alignment=<-1 1 0>, Sprite Facing=<0 1 0>, WriteAlignment=false, WriteFacing=true) {module=SpriteFacingAndAlignment001}
   &CameraOffset(Camera Offset Amount=1)
  ]
  particleUpdateScript [
   &Color(
    Color=<rgba 0.1 0.2 1 1>
    Scale Alpha=&FloatFromCurve(FloatCurve=DataInterface() {propertyValues {Curve {keys [
         <at 0.1 0.2 RCIM_Cubic>
         {interpMode=RCIM_Cubic, tangentMode=RCTM_User, time=0.25, value=0.94}
         {interpMode=RCIM_Cubic, tangentMode=RCTM_User, time=0.5, value=0.55}
         <at 1 0 RCIM_Cubic>
        ]}}})
    Scale Color=&RandomRangeVector(Minimum=<5 5 5>, Maximum=<30 30 30>)
   ) {enabled=false}
   &ParticleState()
   &SpriteRotationRate(Rotation Rate=&RandomRangeFloat(Minimum=-0.1, Maximum=0.1))
   &ScaleSpriteSize(
    Non-Uniform Curve Sprite Scale=DataInterface() {propertyValues {
      XCurve {keys [
        {interpMode=RCIM_Cubic, tangentMode=RCTM_User, time=0, value=1.5}
        {arriveTangent=1, interpMode=RCIM_Cubic, leaveTangent=1, time=0.19, value=2}
        {interpMode=RCIM_Cubic, tangentMode=RCTM_User, time=1, value=2.5}
       ]}
      YCurve {keys [<at 0 1 RCIM_Cubic>, {arriveTangent=0.5, interpMode=RCIM_Cubic, leaveTangent=0.5, time=0.19, value=1.25}, <at 1 1.5 RCIM_Cubic>]}
     }}
    Scale Sprite Size Mode=::NewEnumerator3
   )
  ]
  rendererClasses [&NiagaraSpriteRendererProperties]
  renderers [{rendererClass=&NiagaraSpriteRendererProperties}]
 }
 Rays {
  simTarget=GPUComputeSim
  emitterUpdateScript [
   &EmitterState(Life Cycle Mode=::NewEnumerator1, Loop Duration=1, UseLoopDelay=false)
   &SpawnBurst_Instantaneous(Use Spawn Probability=false, UseLoopCountLimit=true)
  ]
  particleSpawnScript [
   &SkeletalMeshLocation(MaskSpawnGroup=false, OverrideRandomSeed=false, Position Sampling=::NewEnumerator2, Skeletal Mesh=$System.Skeletal Mesh) {enabled=false}
   &InitializeParticle(
    Color=<rgba 0.1 0.2 10 1>
    Position=&SimulationPosition()
    Position Mode=::NewEnumerator1
    Position Offset=<13.5 27 9.5>
    Position Offset Coordinate Space=::Local
    Sprite Size Mode=::NewEnumerator3
    Uniform Sprite Size=16
    UsePositionOffset=true
    Write Material Random=true
   ) {module=InitializeParticle001}
   &SpriteFacingAndAlignment(Facing Coordinate Space=::Local, Sprite Alignment=<-1 1 0>, Sprite Facing=<0 1 0>, WriteAlignment=false, WriteFacing=true) {module=SpriteFacingAndAlignment001}
   &CameraOffset(Camera Offset Amount=5)
  ]
  particleUpdateScript [
   &Color()
   &ParticleState() {enabled=false}
   &SpriteRotationRate(Rotation Rate=0) {enabled=false}
  ]
  rendererClasses [&NiagaraSpriteRendererProperties]
  renderers [{rendererClass=&NiagaraSpriteRendererProperties}]
 }
 Smoke {
  simTarget=GPUComputeSim
  emitterUpdateScript [
   &EmitterState(Life Cycle Mode=::NewEnumerator1, Loop Duration=6, Recalculate Duration Each Loop=true, UseLoopDelay=false)
   &SpawnRate(SpawnRate=10, Use Spawn Probability=false)
  ]
  particleSpawnScript [
   &SkeletalMeshLocation(MaskSpawnGroup=false, OverrideRandomSeed=false, Position Sampling=::NewEnumerator2, Skeletal Mesh=$System.Skeletal Mesh) {enabled=false}
   &InitializeParticle(
    Color Mode=::NewEnumerator0
    Lifetime=&RandomRangeFloat(Minimum=0.5, Maximum=1)
    Position=&SimulationPosition()
    Position Mode=::NewEnumerator1
    Position Offset=<13.5 30 9.5>
    Position Offset Coordinate Space=::Local
    Sprite Rotation Angle=&RandomRangeFloat(Minimum=0, Maximum=1)
    Sprite Rotation Mode=::NewEnumerator1
    Sprite Size=&RandomRangeVector2D(Minimum=<0.32 0.32>, Maximum=<0.64 0.64>)
    Sprite Size Mode=::NewEnumerator4
    UsePositionOffset=true
    Write Material Random=true
   ) {module=InitializeParticle001}
   &CameraOffset(Camera Offset Amount=1)
   &AddVelocity(Velocity=&RandomRangeVector(Minimum=<0 0 0>, Maximum=<0 5 2>))
  ]
  particleUpdateScript [
   /SetVariables(Particles.SubImageIndex=&FloatFromCurve(FloatCurve=DataInterface() {propertyValues {Curve {keys [<at 0 15>,<at 1 31>]}}})) {isSetParameters=true, module=SetVariables_#1}
   &Color(
    Color=<rgba 0.1 0.2 3 1>
    Scale Alpha=&FloatFromCurve(FloatCurve=DataInterface() {propertyValues {Curve {keys [<at 0.5 1>,<at 1 0>]}}})
    Scale Color=<3 3 3>
   ) {enabled=false}
   &ParticleState()
   &SpriteRotationRate(Rotation Rate=&RandomRangeFloat(Minimum=-0.1, Maximum=0.1))
   &ScaleSpriteSize(Uniform Curve Sprite Scale=DataInterface() {propertyValues {Curve {keys [{tangentMode=RCTM_User, time=0, value=20}, {tangentMode=RCTM_User, time=1, value=50}]}}})
   &SolveForcesAndVelocity(Clamp Velocity=false, Limit Acceleration=false)
  ]
  rendererClasses [&NiagaraSpriteRendererProperties]
  renderers [{rendererClass=&NiagaraSpriteRendererProperties}]
 }
}
summary {
 systemName=PS_Wheel_Dire
 emitters [
  {emitterName=Lightnings, simTarget=GPUComputeSim, rendererClasses [&NiagaraSpriteRendererProperties]}
  {emitterName=Smoke, simTarget=GPUComputeSim, rendererClasses [&NiagaraSpriteRendererProperties]}
  {emitterName=Rays, simTarget=GPUComputeSim, rendererClasses [&NiagaraSpriteRendererProperties]}
 ]
 userVariables []
}
